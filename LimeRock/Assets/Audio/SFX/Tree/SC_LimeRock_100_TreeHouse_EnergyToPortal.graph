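# Sound cue for the LimeRock tree house energy-to-portal effect.
# A single wave player feeds the EnergyToPortal wave straight to the output, with no randomization or modulation.
# It plays in the Fort_Events sound class with large point-of-interest attenuation.
# Concurrency overrides cap simultaneous instances per owner and set a retrigger time.
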SoundCue /LimeRock/Assets/Audio/SFX/Tree/SC_LimeRock_100_TreeHouse_EnergyToPortal
root=WavePlayer_1
nodes [{id=WavePlayer_1, properties {soundWaveAssetPtr=/LimeRock/Assets/Audio/SFX/Tree/Wav/LimeRock_100_Tree_EnergyToPortal}}]
properties {
 attenuationSettings=/LimeRock/Assets/Audio/Settings/ATT_LimeRock_POI_Large
 maxDistance=40000
 soundClassObject=&Fort_Events
 volumeMultiplier=1
 concurrencyOverrides {bLimitToOwner=true, maxCount=2, retriggerTime=0.25}
 concurrencySet [null]
}
