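# Sound cue for the looping rift energy sound at the LimeRock tree house.
# A single looping wave player feeds the output with the pink energy glow wave at reduced volume.
# It uses the LimeRock tree attenuation asset and the Fort_Events sound class, with concurrency overrides that cap simultaneous instances per owner.
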SoundCue /LimeRock/Assets/Audio/SFX/Tree/SC_LimeRock_100_TreeHouse_RiftEnergy_LP
root=WavePlayer_1
nodes [{id=WavePlayer_1, properties {bLooping=true, soundWaveAssetPtr=/LimeRock/Assets/Audio/SFX/Tree/Wav/LimeRock_100_Tree_PinkEnergyGlow_LP}}]
properties {
 attenuationSettings=/LimeRock/Assets/Audio/Settings/ATT_LimeRock_Tree
 maxDistance=2097152
 soundClassObject=&Fort_Events
 volumeMultiplier=0.5
 concurrencyOverrides {bLimitToOwner=true, maxCount=2, retriggerTime=0.25}
 concurrencySet [null]
}
