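# Sound cue for the LimeRock tree house being zapped by the rift, judging by its name.
# A single wave player plays the rift gizmo zap wave directly, with slightly reduced volume.
# It uses large point-of-interest attenuation with a long max distance and the Fort_Events sound class.
# Concurrency overrides set a per-owner limit of two instances with a short retrigger time.
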SoundCue /LimeRock/Assets/Audio/SFX/Tree/SC_LimeRock_100_TreeHouse_ZappedByRift
root=WavePlayer_0
nodes [{id=WavePlayer_0, properties {soundWaveAssetPtr=/LimeRock/Assets/Audio/SFX/Tree/Wav/LimeRock_100_Tree_ZappedByTheRiftGizmo}}]
properties {
 attenuationSettings=/LimeRock/Assets/Audio/Settings/ATT_LimeRock_POI_Large
 maxDistance=40000
 soundClassObject=&Fort_Events
 volumeMultiplier=0.7
 concurrencyOverrides {bLimitToOwner=true, maxCount=2, retriggerTime=0.25}
 concurrencySet [null]
}
